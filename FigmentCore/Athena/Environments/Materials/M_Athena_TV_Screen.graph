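# Material for Athena TV screens, showing animated static with optional emergency broadcast and event countdown imagery.
# Static switches enable a broadcast layer and countdown layers, blended by MPC_Countdown's VillainBase and ShowPhase2, with Launch swapping the timer for a weather image.
# The timer splits UVs into eleven cells, picking eight MPC digits from a number atlas with explicit derivatives.
# Launch-text blink, destruction and diffuse-coloration nodes are left unconnected.
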Material /FigmentCore/Athena/Environments/Materials/M_Athena_TV_Screen
graph {
 properties {
  bUseMaterialAttributes=true
  bUsedWithNanite=true
  opacityMaskClipValue=0.16
  pixelDepthOffsetMode=PDOM_Legacy
  usedShadingModels=Default Lit
 }
 nodes [
  n1 = Time()
  n2 = Sine(None=Multiply(A=@n1) {constB=0.3})
  Abs(None=@n2)
  n3 = TextureCoordinate() {coordinateIndex=1}
  n4 = TextureCoordinate() {uTiling=4, vTiling=4}
  n5 = TextureSample(UVs=@n4) {samplerType=SAMPLERTYPE_Normal, texture=/Game/Global/Textures/Default/Tileable_Patterns/BrokenTilingTextureConcrete_N}
  Add(
   A=Multiply(
    A=TextureSample(UVs=@n3) {samplerType=SAMPLERTYPE_Normal, texture=/Game/Global/Textures/Default/Masks_and_Gradients/RevealMask_n}.RGB
    B=Constant3Vector() {constant=<rgba 1 1 0 0>}
   )
   B=@n5.RGB
  )
  n6 = VectorParameter() {group=Diffuse, parameterName=DiffuseColoration, defaultValue=<rgba 1 1 1 0>}
  AppendVector(A=@n6.RGB, B=@n6.RGB)
  n7 = TextureSample(UVs=TextureCoordinate() {uTiling=4, vTiling=4}) {samplerType=SAMPLERTYPE_LinearColor, texture=/Game/Global/Textures/Default/Noise/T_Noise01}
  n8 = TextureSample(UVs=TextureCoordinate() {coordinateIndex=1}) {samplerType=SAMPLERTYPE_LinearColor, texture=/Game/Global/Textures/Default/Masks_and_Gradients/RevealMask}
  n9 = Multiply(A=ConstantBiasScale(None=@n7.R) {bias=-0.1, scale=1}, B=@n8.R)
  n10 = ComponentMask(None=@n5.RGB) {b=true}
  n11 = Multiply(A=Multiply(A=@n8.R, B=OneMinus(None=@n10)), B=Constant() {r=2})
  Clamp(None=Add(A=Add(A=@n9, B=@n11), B=@n8.R))
  Constant() {desc=NumberofDigits, r=8}
  Constant() {r=1}
  n12 = VectorParameter() {parameterName=TextColor, defaultValue=<rgba 1 0 0 1>}
  n13 = MaterialFunctionCall(
   UVs=TextureCoordinate()
   Texture Scale=AppendVector(A=ScalarParameter() {defaultValue=0.6571, parameterName=u}, B=ScalarParameter() {defaultValue=0.2381, parameterName=v})
  ) {materialFunction=&ScaleUVsByCenter}
  n14 = ComponentMask(None=@n13.UVs) {r=true}
  n15 = ComponentMask(None=@n13.UVs) {g=true}
  n16 = Add(A=Clamp(None=Floor(None=@n14)), B=Clamp(None=Floor(None=@n15)))
  n17 = Clamp(None=@n15)
  n18 = Multiply(A=Multiply(A=OneMinus(None=@n16), B=Ceil(None=@n17)), B=Ceil(None=Clamp(None=@n14)))
  n19 = Multiply(A=ComponentMask(None=@n13.UVs) {r=true}) {constB=11}
  n20 = Frac(None=@n19)
  n21 = ComponentMask(None=@n19) {r=true}
  n22 = Constant() {r=9}
  n23 = Constant() {r=6}
  n24 = Constant() {r=3}
  n25 = If(
   A=Constant() {r=2}
   B=@n21
   "A > B"=If(
    A=Constant() {r=1}
    B=@n21
    "A > B"=CollectionParameter() {collection=/Game/Athena/Events/Generic/CountdownTimer/Materials/MPC_Countdown, parameterName=Digit8}
    "A < B"=CollectionParameter() {collection=&MPC_Countdown, parameterName=Digit7}
   )
   "A < B"=Constant() {r=-1}
  )
  n26 = CollectionParameter() {collection=&MPC_Countdown, parameterName=Digit6}
  n27 = If(
   A=Constant() {r=5}
   B=@n21
   "A > B"=If(
    A=Constant() {r=4}
    B=@n21
    "A > B"=If(A=@n24, B=@n21, "A > B"=@n25, "A < B"=@n26)
    "A < B"=CollectionParameter() {collection=&MPC_Countdown, parameterName=Digit5}
   )
   "A < B"=Constant() {r=-1}
  )
  n28 = CollectionParameter() {collection=&MPC_Countdown, parameterName=Digit4}
  n29 = If(
   A=Constant() {r=8}
   B=@n21
   "A > B"=If(
    A=Constant() {r=7}
    B=@n21
    "A > B"=If(A=@n23, B=@n21, "A > B"=@n27, "A < B"=@n28)
    "A < B"=CollectionParameter() {collection=&MPC_Countdown, parameterName=Digit3}
   )
   "A < B"=Constant() {r=-1}
  )
  n30 = CollectionParameter() {collection=&MPC_Countdown, parameterName=Digit2}
  n31 = Floor(None=If(
    A=Constant() {r=10}
    B=@n21
    "A > B"=If(A=@n22, B=@n21, "A > B"=@n29, "A < B"=@n30)
    "A < B"=CollectionParameter() {collection=&MPC_Countdown, parameterName=Digit1}
   ))
  n32 = AppendVector(A=Divide(A=Add(A=@n20, B=@n31)) {constB=10}, B=ComponentMask(None=@n13.UVs) {g=true})
  n33 = TextureCoordinate()
  n34 = DDX(Value=@n33)
  n35 = DDY(Value=@n33)
  n36 = LinearInterpolate(
   A=TextureSample() {texture=/FigmentCore/Athena/Environments/Blueprints/Events/T_TVCountdown_D}.RGB
   B=@n12.RGB
   Alpha=Multiply(
    A=@n18
    B=TextureSample!(UVs=@n32, Tex, "DDX(UVs)"=@n34, "DDY(UVs)"=@n35, Apply View MipBias) {mipValueMode=TMVM_Derivative, samplerType=SAMPLERTYPE_Masks, texture=/Game/Athena/Environments/Blueprints/Events/T_Numbers_M}.R
   )
  )
  n37 = Frac(None=Multiply(A=Time(), B=Constant() {r=3}))
  n38 = Constant() {r=2}
  n39 = TextureObject() {texture=/Game/Global/Textures/Default/PointSplatter/BoatFlake}
  n40 = TextureCoordinate() {uTiling=3, vTiling=3}
  n41 = Add(
   A=Multiply(
    A=MaterialFunctionCall("Animation  Phase (0-1)"=@n37, Number of Rows=@n38, Number of Columns=@n38, Texture=@n39, UVs=@n40) {materialFunction=&FlipBook}.Result
    B=CollectionParameter() {collection=&MPC_Countdown, parameterName=UseStatic}
   )
   B=Multiply(A=Fresnel() {baseReflectFraction=0, exponent=7}) {constB=0.4}
  )
  n42 = Multiply(A=@n41) {constB=0.25}
  n43 = TextureSample!(UVs=@n13.UVs, Tex, Apply View MipBias) {texture=/Game/Athena/Environments/Blueprints/Events/T_LaunchText_M}
  n44 = Time()
  n45 = ScalarParameter() {defaultValue=2.7835, parameterName=BlinkSpeed}
  n46 = Clamp(None=Sine(None=Multiply(A=@n44, B=@n45)))
  n47 = CollectionParameter() {collection=&MPC_Countdown, parameterName=Launch}
  LinearInterpolate(
   A=LinearInterpolate(A=@n36, B=@n12.RGB, Alpha=@n42)
   B=LinearInterpolate(
    A=VectorParameter() {parameterName=ColorLaunch_1, defaultValue=<rgba 0.026 0.0004 0 1>}.RGB
    B=VectorParameter() {parameterName=ColorLaunch_2, defaultValue=<rgba 1 0 0 1>}.RGB
    Alpha=Multiply(A=@n43.RGB, B=@n46)
   )
   Alpha=Multiply(A=Clamp(None=@n18), B=@n47)
  )
  n48 = Multiply(A=@n7.R, B=Constant() {r=0.1})
  n49 = Clamp(None=ScalarParameter() {group=Code_Destruction, parameterName=DestructionAmount})
  n50 = Constant() {r=2}
  n51 = OneMinus(None=MaterialFunctionCall!(Lerp A, Lerp B, Height Texture=Add(A=@n8.R, B=@n48), Falloff Contrast=Constant() {r=2}, Transition Phase=Divide(A=@n49, B=@n50)) {materialFunction=/Game/Packages/Fortnight_MaterialFunctions/Temp/HeightLerp}.Lerp Alpha)
  Multiply(A=@n51, B=@n51)
  ScalarParameter() {defaultValue=1, parameterName=Countdown}
  ScalarParameter() {defaultValue=1, parameterName=VillainBase}
  ScalarParameter() {defaultValue=0.65, group=Diffuse, parameterName=DiffuseColorationThreshold}
  n52 = MakeMaterialAttributes!(
   BaseColor=Constant()
   Metallic
   Specular
   Roughness
   Anisotropy
   EmissiveColor=@n41
   Opacity
   OpacityMask
   Normal
   Tangent
   WorldPositionOffset
   SubsurfaceColor
   ClearCoat
   ClearCoatRoughness
   AmbientOcclusion
   Refraction
   CustomizedUVs_0
   CustomizedUVs_1
   CustomizedUVs_2
   CustomizedUVs_3
   CustomizedUVs_4
   CustomizedUVs_5
   CustomizedUVs_6
   CustomizedUVs_7
   PixelDepthOffset
   ShadingModel
   Displacement
  )
  n53 = TextureSample() {texture=/Game/Athena/Environments/Textures/Screen_WS}
  n54 = MaterialFunctionCall!(A, B, Bounds, Edge Falloff=Constant() {r=70}) {materialFunction=/Engine/Functions/Engine_MaterialFunctions01/Gradient/BoxMask-2D}
  n55 = Abs(None=Sine(None=Multiply(A=@n1) {constB=0.666}))
  n56 = Add(A=Add(A=Multiply(A=@n53.RGB, B=@n54.Result), B=Multiply(A=@n41) {constB=0.5}), B=Multiply(A=Multiply(A=@n2, B=@n55)) {constB=0.2})
  n57 = StaticSwitchParameter(
   True=MaterialFunctionCall!(
    Base Material=@n52
    Top Material=MakeMaterialAttributes(BaseColor=@n53.RGB, EmissiveColor=@n56)
    Alpha=ScalarParameter() {parameterName=Broadcast}
   ) {materialFunction=/Engine/Functions/MaterialLayerFunctions/MatLayerBlend_Standard}.Blended Material
   False=@n52
  ) {parameterName=UseEmergencyBroadCast}
  n58 = Add(A=Multiply(A=@n12.RGB, B=@n42), B=TextureSample() {texture=/Game/Athena/Environments/Blueprints/Events/T_Weather_D}.RGB)
  n59 = MaterialFunctionCall(
   Base Material=@n57
   Top Material=MakeMaterialAttributes(BaseColor=@n58, EmissiveColor=@n58)
   Alpha=CollectionParameter() {collection=&MPC_Countdown, parameterName=VillainBase}
  ) {materialFunction=&MatLayerBlend_Standard}
  n60 = Add(A=@n42, B=LinearInterpolate(A=@n36, B=TextureSample() {texture=/Game/Athena/Environments/Blueprints/Events/T_Weather_3_D}.RGB, Alpha=@n47))
  n61 = MakeMaterialAttributes(BaseColor=@n60, EmissiveColor=@n60)
  n62 = CollectionParameter() {collection=&MPC_Countdown, parameterName=ShowPhase2}
  SetMaterialAttributes!(
   MaterialAttributes=StaticSwitchParameter(
    True=MaterialFunctionCall(Base Material=@n59.Blended Material, Top Material=@n61, Alpha=@n62) {materialFunction=&MatLayerBlend_Standard}.Blended Material
    False=@n57
   ) {defaultValue=true, parameterName=UseCountdown}
   Roughness=Constant() {r=1}
  ) {attributeSetTypes [#1]}
  StaticBoolParameter() {defaultValue=true, group=Diffuse, parameterName=UseProceduralDiffuseColorationMask}
  n63 = TextureSampleParameter2D() {group=Diffuse, parameterName=DiffuseColorationMask, texture=/Game/Global/Textures/Default/Masks_and_Gradients/16x}
  StaticComponentMaskParameter(None=AppendVector(A=@n63.RGB, B=@n63.RGB)) {defaultR=true, group=Diffuse, parameterName=DiffuseColorationMaskComponentMask}
  TextureSample(UVs=@n4) {texture=/Game/Global/Textures/Default/Tileable_Patterns/BrokenTilingTextureConcrete_D}
 ]
}
